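# MF_DecalNormal is a material function in the prototype SprayGun weapon folder that outputs a world-space normal, presumably for decals.
# It reconstructs a flat surface normal from screen-space derivatives, taking DDX and DDY of world position.
# Both derivatives are normalized and crossed to give WorldNormal, with no normalization after the cross product.
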MaterialFunction /Game/Weapons/Prototype/SprayGun/Materials/MF_DecalNormal
graph {nodes [
  n1 = WorldPosition()
  n2 = DDX(Value=@n1.XYZ)
  n3 = DDY(Value=@n1.XYZ)
  FunctionOutput(None=CrossProduct(A=Normalize(VectorInput=@n2), B=Normalize(VectorInput=@n3))) {outputName=WorldNormal}
 ]}
