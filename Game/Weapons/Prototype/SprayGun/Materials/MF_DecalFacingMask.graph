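# Material function that outputs a facing (slope) mask for spray gun decals, fading them on surfaces angled away from the projection, described as Jordan's slope mask from Paragon.
# By default it negates the dot product of the MF_DecalNormal world normal and object orientation, then remaps and clamps it for higher contrast.
# A Custom HLSL preprocessor switch substitutes a near-binary camera-vector versus orientation test when compiling for GLSL ES2/ES3.1 (Android).
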MaterialFunction /Game/Weapons/Prototype/SprayGun/Materials/MF_DecalFacingMask
graph {
 properties {bExposeToLibrary=true, description=Jordan's slope mask from Paragon}
 nodes [
  n1 = Multiply(A=DotProduct(A=MaterialFunctionCall() {materialFunction=&/MF_DecalNormal}.WorldNormal, B=ObjectOrientation())) {constB=-1}
  MaterialExpressionClamp_0 = Clamp(None=LinearInterpolate(Alpha=Clamp(None=@n1)) {constA=-0.5, constB=1.5})
  n2 = CameraVectorWS()
  n3 = ObjectOrientation()
  MaterialExpressionSaturate_0 = Saturate(None=Multiply(A=DotProduct(A=@n2, B=@n3)) {constB=-100})
  n4 = Custom!(Default=@MaterialExpressionClamp_0, Android=@MaterialExpressionSaturate_0) {
   code="##if COMPILER_GLSL_ES2 || COMPILER_GLSL_ES3_1\r\n\treturn Android;\r\n##else\r\n\treturn Default;\r\n##endif"
   description=AndroidSwitch
   outputType=CMOT_Float1
   inputs [
    {inputName=Default, input {expression=&self:MaterialExpressionClamp_0}}
    {inputName=Android, input {expression=&self:MaterialExpressionSaturate_0}}
   ]
  }
  FunctionOutput(None=@n4)
 ]
}
